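; Gameplay ability for MotoWrestle that detects the end of an icy slide and cleans up its effects on the player.
; On activation it caches the avatar's Fort player pawn and waits for the MotoWrestle.Slide.Icy tag to be removed from it.
; It then removes effects granting the IceGrenade IcyFeet tag and sets the capsule's response to GameTraceChannel18 to Overlap.
; The graph never calls EndAbility, and its end event is empty.
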(event Ability|EventActivateAbility
  (bind _outfortplayerpawn (CallFunction|GetAvatarasValidFortPlayerPawn)
    (:ValidFortPlayerPawn
      (bind _output_get (Variables|Default|SetOut%20Fort%20Player%20Pawn _outfortplayerpawn))
      (Ability|Tasks|WaitGameplayTagRemoveFromActor _output_get "(TagName=\"MotoWrestle.Slide.Icy\")"
        (:Removed
          (Ability|RemoveGameplayEffectFromOwnerWithGrantedTags "(GameplayTags=((TagName=\"GameplayEffect.Item.IceGrenade.IcyFeet\")))")
          (Collision|SetCollisionResponsetoChannel (Variables|/Script/Engine.Character|GetCapsuleComponent (Variables|Default|GetOut%20Fort%20Player%20Pawn)) "ECC_GameTraceChannel18" "ECR_Overlap"))))))

(event Ability|EventOnEndAbility
  (params (bWasCancelled :type Boolean)))
